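# Translucent, two-sided unlit fire material for the male Gimmick backpack effect, usable on Niagara mesh particles.
# Opacity sums two samples of a flame mask warped by panning swirl and smoke noise, with an optional detail mask and depth fade outside forward and mobile paths.
# Emissive blends curve-atlas color and base color through partly rounded scrolling noise, tinted by a radial mask and particle color.
# A panning distortion texture drives optional local-Z vertex wobble.
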Material /BRCosmetics/Accessories/FORT_Backpacks/Backpack_M_MED_Gimmick/FX/M_Backpack_Gimmick_Male_Fire
graph {
 properties {
  bUsedWithNiagaraMeshParticles=true
  blendMode=BLEND_Translucent
  pixelDepthOffsetMode=PDOM_Legacy
  shadingModel=MSM_Unlit
  twoSided=true
  usedShadingModels=Unlit
 }
 nodes [
  n1 = Constant3Vector() {constant=<rgba 0 0 1 0>}
  n2 = TextureCoordinate()
  n3 = Multiply(A=Time(), B=ScalarParameter() {defaultValue=2, group=WPO, parameterName=WPO_Speed})
  n4 = Multiply(A=TextureSample(UVs=Panner(Coordinate=@n2, Time=@n3) {speedX=0.5, speedY=0.5}) {samplerType=SAMPLERTYPE_Normal, texture=&T_Weap_Distort_N_02}.R) {constB=1.2}
  LinearInterpolate(
   A=Constant()
   B=Transform(None=Multiply(A=@n1, B=@n4)) {transformSourceType=TRANSFORMSOURCE_Local}
   Alpha=ScalarParameter() {defaultValue=1, group=WPO, parameterName=Use_WPO}
  )
  n5 = CurveAtlasRowParameter(CurveTime=ComponentMask(None=TextureCoordinate()) {r=true}) {
   atlas=/Game/Athena/LinearColorCurves/VFX_CurveAtlas
   curve=/Game/Weapons/FORT_Melee/Pickaxe_Cube_Queen_Female/FX/Cube_Queen_EmissiveCurve
   defaultValue=41
   group=Emissive
  }
  n6 = Multiply(A=@n5) {constB=2}
  n7 = VectorParameter() {group=Emissive, parameterName=BaseColor, defaultValue=<rgba 5 0 2.0646 1>}
  n8 = Panner(Coordinate=TextureCoordinate() {uTiling=2, vTiling=0.4}) {speedY=1}
  n9 = TextureCoordinate() {vTiling=0.5}
  n10 = Multiply(A=TextureSample(UVs=Panner(Coordinate=@n9) {speedY=0.2}) {texture=&T_Inky_Smoke_Tile}.R) {constB=0.2}
  n11 = LinearInterpolate(Alpha=TextureSample(UVs=Add(A=@n8, B=@n10)) {texture=/Game/Effects/Fort_Effects/Textures/Power_Source/T_Power_Source_Sphere_Texture}.R) {constA=0.1}
  n12 = LinearInterpolate(A=Round(None=@n11), B=@n11, Alpha=Constant() {r=0.8}) {constAlpha=0.8}
  n13 = SphereMask(A=TextureCoordinate(), B=Constant2Vector() {g=0.5, r=0.5}) {attenuationRadius=0.5, hardnessPercent=0}
  n14 = Multiply(
   A=Multiply(A=LinearInterpolate(A=@n6, B=@n7.RGB, Alpha=@n12)) {constB=10}
   B=LinearInterpolate(
    A=VectorParameter() {group=Emissive, parameterName=SphereMask_Color_Dark, defaultValue=<rgba 0.2 0.0054 0.0558 1>}.RGB
    B=VectorParameter() {group=Emissive, parameterName=SphereMask_Color_Birght, defaultValue=<rgba 3 0 0.5035 1>}.RGB
    Alpha=Power(Base=@n13)
   ) {constB=3}
  )
  n15 = Constant() {r=1}
  n16 = ScalarParameter() {group=Emissive, parameterName=Use_Simple_Emissive}
  MaterialFunctionCall(Emissive=Multiply(A=ParticleColor().RGB, B=LinearInterpolate(A=@n14, B=@n15, Alpha=@n16))) {materialFunction=&MF_FX_EmissiveAdjustments}
  n17 = MaterialFunctionCall(UVs=TextureCoordinate(), "Rotation Angle (0-1)"=Constant() {r=-0.03}) {materialFunction=&CustomRotator}
  n18 = Constant2Vector() {g=0.7, r=2.5}
  n19 = TextureSample(UVs=Panner(Coordinate=Multiply(A=@n17.Rotated Values, B=@n18)) {speedX=-0.3, speedY=0.5}) {samplerType=SAMPLERTYPE_LinearGrayscale, texture=/Game/Global/Textures/Default/Noise/T_SwirlNoise_N_height}
  n20 = OneMinus(None=ComponentMask(None=TextureCoordinate()) {r=true})
  n21 = Constant() {r=0.5}
  n22 = Saturate(None=Multiply(A=SmoothStep(Min=@n20, Value=@n21)) {constB=5})
  n23 = Add(A=ComponentMask(None=TextureCoordinate()) {r=true}, B=Multiply(A=LinearInterpolate(B=@n19.R, Alpha=@n22)) {constB=0.1}) {constB=0}
  n24 = ComponentMask(None=TextureCoordinate() {uTiling=2.2}) {r=true}
  n25 = TextureCoordinate() {uTiling=3.5, vTiling=0.5}
  n26 = ComponentMask(None=@n25) {r=true}
  n27 = TextureCoordinate()
  n28 = Power(Base=OneMinus(None=ComponentMask(None=@n27) {r=true})) {constExponent=0.5}
  n29 = LinearInterpolate(
   A=TextureCoordinate() {uTiling=2, vTiling=0.5}
   B=AppendVector(A=LinearInterpolate(A=@n24, B=@n26, Alpha=@n28), B=ComponentMask(None=@n25) {g=true})
   Alpha=ScalarParameter() {defaultValue=1, group=Opacity, parameterName=Use_U_Mask}
  )
  n30 = TextureCoordinate()
  n31 = Power(Base=OneMinus(None=ComponentMask(None=@n30) {g=true})) {constExponent=0.5}
  n32 = ComponentMask(None=TextureCoordinate()) {r=true}
  n33 = Saturate(None=Multiply(A=Power(Base=@n32) {constExponent=7}) {constB=5})
  n34 = Multiply(
   A=TextureSample(UVs=Panner(Coordinate=@n29) {speedX=0.3, speedY=1}) {texture=&T_Inky_Smoke_Tile}.R
   B=LinearInterpolate(B=@n31, Alpha=Multiply(A=@n33, B=@n31)) {constA=1}
  )
  n35 = Add(A=ComponentMask(None=TextureCoordinate()) {g=true}, B=Multiply(A=LinearInterpolate(Alpha=@n34) {constA=-1, constB=0.5}) {constB=0.3})
  n36 = TextureSample(UVs=Add(A=AppendVector(A=@n23, B=@n35), B=Constant2Vector() {g=0.25})) {
   samplerSource=SSM_Clamp_WorldGroupSettings
   texture=&/T_Gimmick_Backpack_FX_Mask
  }
  n37 = TextureCoordinate()
  n38 = Constant() {r=0.03}
  n39 = Panner(Coordinate=Multiply(
    A=MaterialFunctionCall(UVs=@n37, "Rotation Angle (0-1)"=@n38) {materialFunction=&CustomRotator}.Rotated Values
    B=Constant2Vector() {g=0.7, r=2.5}
   )) {speedX=0.3, speedY=0.5}
  n40 = Add(
   A=ComponentMask(None=TextureCoordinate()) {r=true}
   B=Multiply(A=TextureSample(UVs=@n39) {samplerType=SAMPLERTYPE_LinearGrayscale, texture=&T_SwirlNoise_N_height}.R) {constB=0.1}
  ) {constB=0}
  n41 = Panner(Coordinate=TextureCoordinate() {uTiling=3, vTiling=-0.5}) {speedX=0.15, speedY=1}
  n42 = TextureSample(UVs=Panner(Coordinate=TextureCoordinate() {uTiling=3, vTiling=0}) {speedX=0.15}) {samplerType=SAMPLERTYPE_Grayscale, texture=&LowResBlurredNoise}
  n43 = Multiply(A=Multiply(
    A=TextureSample(UVs=@n41) {texture=&T_Inky_Smoke_Tile}.R
    B=Clamp(None=@n42.R) {maxDefault=0.3}
   )) {constB=-1}
  n44 = Add(A=ComponentMask(None=TextureCoordinate()) {g=true}, B=Multiply(A=LinearInterpolate(Alpha=@n43) {constA=-0.6, constB=0.5}) {constB=0.5})
  n45 = TextureSample(UVs=Add(A=AppendVector(A=@n40, B=@n44), B=Constant2Vector() {g=0.2})) {
   samplerSource=SSM_Clamp_WorldGroupSettings
   texture=&T_Gimmick_Backpack_FX_Mask
  }
  n46 = Add(A=Add(A=@n36.R, B=@n45.R)) {constB=0}
  n47 = LinearInterpolate(Alpha=Saturate(None=Add(A=@n46) {constB=0})) {constA=-0.2}
  n48 = TextureCoordinate() {uTiling=1.5, vTiling=0.6}
  n49 = TextureCoordinate() {vTiling=0.5}
  n50 = Multiply(A=TextureSample(UVs=Panner(Coordinate=@n49) {speedY=0.1}) {texture=&T_Inky_Smoke_Tile}.R) {constB=0.2}
  n51 = TextureSample(UVs=Add(A=Panner(Coordinate=@n48) {speedY=1}, B=@n50)) {texture=&T_Inky_Smoke_Tile}
  n52 = TextureCoordinate() {uTiling=1.5, vTiling=0.5}
  n53 = TextureSample(UVs=Add(A=@n50, B=Panner(Coordinate=@n52) {speedY=0.3})) {samplerType=SAMPLERTYPE_LinearColor, texture=/Game/Effects/Fort_Effects/BluGlow/Violentium/Violentium_Mask}
  n54 = LinearInterpolate(Alpha=Multiply(A=Multiply(A=@n51.R, B=@n53.R), B=ScalarParameter() {defaultValue=2, parameterName=Dark flame range}) {constB=2})
  n55 = Round(None=@n54)
  n56 = Constant() {r=0.6}
  n57 = Multiply(
   A=Multiply(A=ParticleColor().A, B=Saturate(None=@n47))
   B=LinearInterpolate(
    A=Constant() {r=1}
    B=LinearInterpolate(A=@n55, B=@n54, Alpha=@n56) {constAlpha=0.8}
    Alpha=ScalarParameter() {group=Opacity, parameterName=Use_Detail_Mask}
   )
  )
  n58 = DepthFade(Opacity=@n57) {fadeDistanceDefault=5}
  Reroute(None=ShadingPathSwitch(Default=@n58, Deferred=@n58, Forward=@n57, Mobile=@n57))
  StaticComponentMaskParameter(None=@n45.RGB) {defaultR=true, group=Opacity, parameterName=Component Mask}
  StaticComponentMaskParameter(None=@n36.RGB) {defaultR=true, group=Opacity, parameterName=Component Mask}
 ]
}
